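; Gameplay ability for a Battle Pass S20 quest, named for explosives while sliding, that reports C4 detonations as quest progress.
; On event activation it calls the parent, then waits for C4 explode gameplay events on the owning pawn's player controller.
; Each received event sends a count of one to the owning player's quest manager under the quest tag variable.
; The graph itself contains no sliding check.
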(event Ability|EventActivateAbilityFromEvent
  (params (EventData :type Gameplay%20Event%20Data%20Structure%20%28by%20ref%29))
  (|CallParentFunction EventData)
  (Ability|Tasks|WaitGameplayEvent "(TagName=\"Abilities.Generic.C4.Explode\")" (Fortnite|GetFortPlayerControllerfromActor (Variables|GAVars|Data|GetOwningPlayerPawn))
    (:EventReceived
      (Quest|SendComplexCustomStatEvent (Variables|GAVars|Data|GetOwningPlayerQuestManager) 0 (Variables|Default|GetT_Quest) "(GameplayTags=)" 1))))
